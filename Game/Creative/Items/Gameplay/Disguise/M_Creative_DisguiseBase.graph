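# Additive particle sprite material for the Creative disguise item that draws a glitchy hologram effect over the scene behind it.
# On the deferred path it reads scene normals and position behind the sprite for fresnel and horizontal lines, adding panned mote noise tinted by a base color.
# Tinted data-chip blocks pan in stepped time, boosted by an automatic or dynamic-parameter-driven height sweep.
# Depth distance, sweep height and UV-edge masks bound the result, with depth fade on deferred.
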Material /Game/Creative/Items/Gameplay/Disguise/M_Creative_DisguiseBase
graph {
 properties {
  bEnableResponsiveAA=true
  bUsedWithNiagaraSprites=true
  bUsedWithParticleSprites=true
  blendMode=BLEND_Additive
  pixelDepthOffsetMode=PDOM_Legacy
  usedShadingModels=Default Lit
 }
 nodes [
  Constant()
  n1 = ComponentMask(None=SceneTexture() {sceneTextureId=PPI_WorldNormal}.Color) {b=true, g=true, r=true}
  n2 = ScalarParameter() {parameterName=Use_UpdatedFresnel}
  n3 = MaterialFunctionCall(
   Brightness=ScalarParameter() {defaultValue=1, parameterName=FresnelBrightness}
   Exponent=ScalarParameter() {defaultValue=3, parameterName=FresnelExponent}
   Camera Vector=LinearInterpolate(A=@n1, B=CameraVectorWS(), Alpha=@n2)
   Vertex / Pixel Normal=LinearInterpolate(A=VertexNormalWS(), B=@n1, Alpha=@n2)
  ) {materialFunction=&MF_Fresnel_FX}
  n4 = Constant()
  n5 = ShadingPathSwitch(Default=@n3.Fresnel Inverted, Deferred=@n3.Fresnel Inverted, Forward=@n4, Mobile=@n4)
  n6 = ParticleMacroUV()
  n7 = Time()
  n8 = ScalarParameter() {defaultValue=8, parameterName=SweepSteps}
  n9 = Divide(A=Ceil(None=Multiply(A=@n7, B=@n8)), B=@n8)
  n10 = Sine(None=@n9) {period=0.4575}
  n11 = Sine(None=@n9) {period=0.705}
  n12 = VertexInterpolator(VS=ConstantBiasScale(None=Add(A=@n10, B=@n11)) {bias=2, scale=0.25})
  n13 = VectorParameter() {parameterName=NoisePanSpeed, defaultValue=<rgba 0.075 0.03 -0.111 -0.096>}
  n14 = AppendVector(A=@n13.R, B=@n13.R)
  n15 = AppendVector(A=@n13.B, B=@n13.B)
  n16 = Multiply(
   A=TextureSample(UVs=Panner(Coordinate=@n6, Time=@n12.PS, Speed=@n14) {bFractionalPart=true}) {texture=/Game/Global/Textures/Default/Noise/TankMoteNoise}.R
   B=TextureSample(UVs=Panner(Coordinate=@n6, Time=@n12.PS, Speed=@n15) {bFractionalPart=true}) {texture=&TankMoteNoise}.R
  )
  n17 = Multiply(A=@n16, B=ScalarParameter() {defaultValue=0.1, parameterName=NoiseOpacity})
  n18 = MaterialFunctionCall() {materialFunction=&WorldPositionBehindTranslucency}
  n19 = WorldPosition() {worldPositionShaderOffset=WPT_ExcludeAllShaderOffsets}
  n20 = Divide(
   A=ComponentMask(None=ShadingPathSwitch(Default=@n18.Result, Deferred=@n18.Result, Forward=@n19.XYZ, Mobile=@n19.XYZ)) {b=true}
   B=ScalarParameter() {defaultValue=4, parameterName=LineHeightZDivisor}
  )
  n21 = Power(
   Base=SphereMask(
    A=Frac(None=@n20)
    B=Constant() {r=0.5}
    Radius=ScalarParameter() {defaultValue=0.15, parameterName=LinesRadius}
    Hardness=ScalarParameter() {defaultValue=0.2, parameterName=LinesHardness}
   )
   Exp=ScalarParameter() {defaultValue=8, parameterName=LineExponent}
  )
  n22 = Multiply(
   A=Add(A=Add(A=@n5, B=@n17), B=Multiply(A=@n16, B=@n21))
   B=VectorParameter() {parameterName=BaseColor, defaultValue=<rgba 0 0.3546 1.8 1>}.RGB
  )
  n23 = NamedRerouteUsage()
  n24 = VectorParameter() {parameterName=Red, defaultValue=<rgba 0.275 0.0422 0.5 0>}
  n25 = Multiply(A=TextureCoordinate(), B=DynamicParameter() {paramNames [UVScale, TimeSpeed, BlockOpacity, ManuallyControlSweepPosition]}.UVScale)
  n26 = Multiply(A=DynamicParameter() {paramNames [UVScale, TimeSpeed, BlockOpacity, ManuallyControlSweepPosition]}.TimeSpeed, B=Time())
  n27 = ScalarParameter() {defaultValue=6, parameterName=Steps}
  n28 = Divide(A=Ceil(None=Multiply(A=@n26, B=@n27)), B=@n27)
  n29 = VectorParameter() {parameterName=BlocksPanSpeed, defaultValue=<rgba 1.8 0.1 -0.8 -0.62>}
  n30 = TextureSample(UVs=Panner(Coordinate=@n25, Time=@n28, Speed=AppendVector(A=@n29.R, B=@n29.R)) {bFractionalPart=true}) {texture=/Game/Athena/UI/Challenges/Art/T_UI_DataChip_Noise}
  n31 = WorldPosition()
  n32 = Divide(
   A=ComponentMask(None=TransformPosition(None=@n31.XYZ) {transformSourceType=TRANSFORMPOSSOURCE_World}) {b=true}
   B=ScalarParameter() {defaultValue=170, parameterName=GlitchHeightZDivisor}
  ) {constB=300}
  n33 = Constant()
  n34 = DynamicParameter() {paramNames [UVScale, TimeSpeed, BlockOpacity, ManuallyControlSweepPosition]}
  n35 = Multiply(
   A=SphereMask(
    A=@n32
    B=If(
     A=@n33
     B=@n34.ManuallyControlSweepPosition
     "A > B"=@n12.PS
     "A == B"=@n34.ManuallyControlSweepPosition
     "A < B"=@n34.ManuallyControlSweepPosition
    )
    Radius=ScalarParameter() {defaultValue=0.06, parameterName=SweepRadius}
    Hardness=ScalarParameter() {defaultValue=0.8, parameterName=SweepHardness}
   )
   B=ScalarParameter() {defaultValue=1.5, parameterName=SweepScale}
  )
  n36 = Add(A=@n30.R, B=Multiply(A=@n30.R, B=@n35))
  n37 = VectorParameter() {parameterName=Green, defaultValue=<rgba 0.0242 0.5437 0.58 0>}
  n38 = TextureSample(UVs=Panner(Coordinate=@n25, Time=@n28, Speed=AppendVector(A=@n29.B, B=@n29.B)) {bFractionalPart=true}) {texture=&T_UI_DataChip_Noise}
  n39 = Add(A=@n38.G, B=Multiply(A=@n38.G, B=@n35))
  n40 = Power(Base=Add(A=Multiply(A=@n24.RGB, B=@n36), B=Multiply(A=@n37.RGB, B=@n39)), Exp=ScalarParameter() {defaultValue=3, parameterName=BlocksPower})
  n41 = ScalarParameter() {defaultValue=-1, parameterName=BlocksDesaturation}
  n42 = Add(
   A=LinearInterpolate(A=@n22, B=Multiply(A=@n23, B=@n22), Alpha=ScalarParameter() {parameterName=Use_UpdatedFresnel})
   B=Multiply(
    A=Desaturation(None=@n40, Fraction=@n41)
    B=DynamicParameter() {paramNames [UVScale, TimeSpeed, BlockOpacity, ManuallyControlSweepPosition]}.BlockOpacity
   )
  )
  n43 = MaterialFunctionCall!(Distance Bias, Distance Divisor=ScalarParameter() {defaultValue=50, parameterName=Distance Divisor}, Scene Depth, Pixel Depth) {materialFunction=/Game/Effects/Fort_Effects/Materials/_Functions/MF_DistanceBehindTranslucency_01}
  n44 = Constant()
  n45 = Add(
   A=DynamicParameter() {paramNames [UVScale, TimeSpeed, BlockOpacity, ManuallyControlSweepPosition]}.ManuallyControlSweepPosition
   B=ScalarParameter() {defaultValue=0.2, parameterName=EndSweepMaskBias}
  )
  n46 = If(A=@n44, B=@n45, "A > B"=Constant() {r=1}, "A == B"=@n45, "A < B"=@n45)
  n47 = Constant() {r=0.9}
  n48 = Abs(None=ConstantBiasScale(None=TextureCoordinate()) {bias=-0.5, scale=2})
  n49 = Constant()
  n50 = Multiply(
   A=Multiply(A=Multiply(A=@n42, B=@n43.Normalized Inverted), B=SphereMask(A=@n32, B=@n44, Radius=@n46, Hardness=@n47))
   B=SphereMask(A=Power(Base=@n48) {constExponent=4}, B=@n49, Radius=Constant() {r=1}, Hardness=@n49)
  )
  n51 = DepthFade(Opacity=@n50, FadeDistance=ScalarParameter() {defaultValue=30, parameterName=FadeDistance})
  MaterialFunctionCall(Emissive=ShadingPathSwitch(Default=@n51, Deferred=@n51, Forward=@n50, Mobile=@n50)) {materialFunction=&MF_FX_EmissiveAdjustments}
  NamedRerouteDeclaration(None=@n5) {name=FresnelOutput, nodeColor=<rgba 0.3828 0 1 1>}
  ParticleColor()
 ]
}
