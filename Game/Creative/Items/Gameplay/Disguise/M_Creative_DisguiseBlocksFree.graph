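# Additive unlit particle material for the Creative disguise item, drawing animated noise blocks on Cascade or Niagara sprites.
# Two samples of a data-chip noise texture pan at different speeds on ceiling-quantized time, so movement steps, and their red and green channels multiply into the pattern.
# Particle color tints it through an emissive-adjustment function.
# Scalar switches blend opacity between constant, pattern and particle alpha, while dynamic parameters drive UV scale and speed.
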Material /Game/Creative/Items/Gameplay/Disguise/M_Creative_DisguiseBlocksFree
graph {
 properties {
  bEnableResponsiveAA=true
  bUsedWithNiagaraSprites=true
  bUsedWithParticleSprites=true
  blendMode=BLEND_Additive
  pixelDepthOffsetMode=PDOM_Legacy
  shadingModel=MSM_Unlit
  usedShadingModels=Unlit
 }
 nodes [
  n1 = Multiply(A=TextureCoordinate(), B=DynamicParameter() {paramNames [UVScale,TimeSpeed,BlockOpacity,Param4]}.UVScale)
  n2 = Multiply(A=DynamicParameter() {paramNames [UVScale,TimeSpeed,BlockOpacity,Param4]}.TimeSpeed, B=Time())
  n3 = ScalarParameter() {defaultValue=6, parameterName=Steps}
  n4 = Divide(A=Ceil(None=Multiply(A=@n2, B=@n3)), B=@n3)
  n5 = VectorParameter() {parameterName=BlocksPanSpeed, defaultValue=<rgba 1.8 0.1 -0.8 -0.62>}
  n6 = AppendVector(A=@n5.R, B=@n5.R)
  n7 = AppendVector(A=@n5.B, B=@n5.B)
  n8 = Multiply(
   A=TextureSample(UVs=Panner(Coordinate=@n1, Time=@n4, Speed=@n6) {bFractionalPart=true}) {texture=/Game/Athena/UI/Challenges/Art/T_UI_DataChip_Noise}.R
   B=TextureSample(UVs=Panner(Coordinate=@n1, Time=@n4, Speed=@n7) {bFractionalPart=true}) {texture=&T_UI_DataChip_Noise}.G
  )
  n9 = LinearInterpolate(A=Constant() {r=1}, B=@n8, Alpha=ScalarParameter() {parameterName=UseOpacityFromEmissive})
  n10 = ParticleColor()
  LinearInterpolate(
   A=@n9
   B=LinearInterpolate(A=Multiply(A=@n9, B=@n10.A), B=@n9, Alpha=ScalarParameter() {parameterName=MultiplyParticleAlpha})
   Alpha=ScalarParameter() {parameterName=Use Particle Alpha}
  )
  MaterialFunctionCall(Emissive=Multiply(A=@n10.RGB, B=@n8)) {materialFunction=&MF_FX_EmissiveAdjustments}
 ]
}
