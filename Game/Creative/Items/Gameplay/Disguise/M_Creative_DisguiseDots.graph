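# Additive unlit particle material for the Creative Disguise item, drawing a grid of glowing dots that vary in brightness over time.
# Particle size divided by RepeatUnits and rounded sets the UV tiling, so dot count scales with sprite size.
# A squared sphere mask shapes each dot, multiplied by exponent-sharpened random UV noise and particle color, then passed through emissive adjustments.
# Dynamic parameters drive dot radius, random seed and animation speed.
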Material /Game/Creative/Items/Gameplay/Disguise/M_Creative_DisguiseDots
graph {
 properties {
  bEnableResponsiveAA=true
  bUsedWithNiagaraSprites=true
  bUsedWithParticleSprites=true
  blendMode=BLEND_Additive
  pixelDepthOffsetMode=PDOM_Legacy
  shadingModel=MSM_Unlit
  usedShadingModels=Unlit
 }
 nodes [
  n1 = ParticleSize()
  n2 = ScalarParameter() {defaultValue=2, parameterName=RepeatUnits}
  n3 = Multiply(A=TextureCoordinate(), B=Round(None=Divide(A=@n1, B=@n2)))
  n4 = SphereMask(
   A=Frac(None=@n3)
   B=Constant() {r=0.5}
   Radius=DynamicParameter() {paramNames [BlocksRandomSeed, BlocksTimeScale, BlocksSize, Param4]}.BlocksSize
   Hardness=ScalarParameter() {defaultValue=0.2, parameterName=Hardness}
  )
  n5 = Multiply(A=@n4, B=@n4)
  n6 = DynamicParameter() {paramNames [BlocksRandomSeed, BlocksTimeScale, BlocksSize, Param4]}
  n7 = Multiply(A=Time(), B=DynamicParameter() {paramNames [BlocksRandomSeed, BlocksTimeScale, BlocksSize, Param4]}.BlocksTimeScale)
  n8 = Power(
   Base=MaterialFunctionCall!(
    PrimaryTextureCoordinates=@n3
    2ndOrderTextureCoordinates
    RandomSeed=Add(A=@n6.BlocksRandomSeed, B=@n7)
    NoiseDensityBias
    Bypass Noise Influence Crosstalk
   ) {materialFunction=/Game/Effects/Fort_Effects/Materials/_Functions/MF_RandomUVSample}.NoiseWithRange
   Exp=ScalarParameter() {defaultValue=8, parameterName=BlocksExponent}
  )
  MaterialFunctionCall(Emissive=Multiply(A=ParticleColor().RGB, B=Multiply(A=@n5, B=@n8))) {materialFunction=&MF_FX_EmissiveAdjustments}
 ]
}
