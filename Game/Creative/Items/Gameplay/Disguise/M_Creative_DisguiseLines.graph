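# Additive, unlit, two-sided material for glowing lines on mesh particles, evidently part of the Creative disguise item's effect.
# SplineThicken sets line width between tip and base, and a world offset weighted by inverse V adds jitter from two summed sines on step-quantized time.
# Emissive combines a U-based sphere mask, a V-power gradient, a color parameter and particle alpha, then passes through MF_FX_EmissiveAdjustments.
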Material /Game/Creative/Items/Gameplay/Disguise/M_Creative_DisguiseLines
graph {
 properties {
  bUsedWithMeshParticles=true
  blendMode=BLEND_Additive
  pixelDepthOffsetMode=PDOM_Legacy
  shadingModel=MSM_Unlit
  twoSided=true
  usedShadingModels=Unlit
 }
 nodes [
  n1 = MaterialFunctionCall(Float2=TextureCoordinate()) {materialFunction=&BreakOutFloat2Components}
  n2 = Time()
  n3 = ScalarParameter() {defaultValue=8, parameterName=SweepSteps}
  n4 = Divide(A=Ceil(None=Multiply(A=@n2, B=@n3)), B=@n3)
  n5 = Multiply(A=Add(A=Sine(None=@n4) {period=0.4575}, B=Sine(None=@n4) {period=0.705})) {constB=0.5}
  n6 = VectorParameter() {parameterName=WPO, defaultValue=<rgba 0 0 170 0.2>}
  n7 = Multiply(A=@n6.RGB, B=@n6.RGB)
  Add(
   A=MaterialFunctionCall(
    WidthBase=ScalarParameter() {defaultValue=1, parameterName=ThicknessTip}
    WidthTip=ScalarParameter() {defaultValue=10, parameterName=ThicknessBase}
   ) {materialFunction=&SplineThicken}.WorldPositionOffset
   B=Multiply(A=OneMinus(None=@n1.G), B=Multiply(A=@n5, B=@n7))
  )
  Constant() {r=0.55}
  n8 = DynamicParameter() {paramNames [UVScale, TimeSpeed, BlockOpacity, ManuallyControlSweepPosition]}
  If(A=Constant(), B=@n8.ManuallyControlSweepPosition, "A == B"=@n8.ManuallyControlSweepPosition, "A < B"=@n8.ManuallyControlSweepPosition)
  n9 = Power(Base=Abs(None=ConstantBiasScale(None=@n1.R) {bias=-0.5, scale=2})) {constExponent=4}
  n10 = Constant() {r=0.5}
  n11 = Constant()
  n12 = Power(Base=@n1.G, Exp=ScalarParameter() {defaultValue=3, parameterName=Exponent})
  n13 = Multiply(
   A=LinearInterpolate(A=SphereMask(A=@n9, B=@n10, Radius=@n10, Hardness=@n11), Alpha=@n12) {constA=1}
   B=VectorParameter() {parameterName=Color, defaultValue=<rgba 0 0.72 2 0>}.RGB
  )
  MaterialFunctionCall(Emissive=Multiply(A=Multiply(A=@n13, B=@n12), B=ParticleColor().A)) {materialFunction=&MF_FX_EmissiveAdjustments}
 ]
}
