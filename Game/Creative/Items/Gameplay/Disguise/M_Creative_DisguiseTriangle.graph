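# Additive unlit sprite material for particle effects, named for the triangle shapes of the Creative disguise item.
# The triangle texture's red channel is scaled by a CenterScale dynamic parameter, added to its alpha channel, then tinted by particle color RGB.
# The result feeds the emissive input of MF_FX_EmissiveAdjustments, letting each particle system drive the red-channel intensity separately from the alpha shape.
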Material /Game/Creative/Items/Gameplay/Disguise/M_Creative_DisguiseTriangle
graph {
 properties {
  bUsedWithNiagaraSprites=true
  bUsedWithParticleSprites=true
  blendMode=BLEND_Additive
  pixelDepthOffsetMode=PDOM_Legacy
  shadingModel=MSM_Unlit
  usedShadingModels=Unlit
 }
 nodes [
  n1 = TextureSample() {samplerType=SAMPLERTYPE_LinearColor, texture=&/T_DishFries_Triangle}
  n2 = Multiply(A=DynamicParameter() {paramNames [CenterScale,Param2,Param3,Param4]}.CenterScale, B=@n1.R)
  MaterialFunctionCall(Emissive=Multiply(A=Add(A=@n2, B=@n1.A), B=ParticleColor().RGB)) {materialFunction=&MF_FX_EmissiveAdjustments}
 ]
}
